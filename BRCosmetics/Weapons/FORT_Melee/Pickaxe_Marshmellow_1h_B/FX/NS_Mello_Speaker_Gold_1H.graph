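# Niagara effect for the gold one-handed Mello (Marshmellow) pickaxe, emitting short-lived, speaker-style waveform meshes.
# A single Waveforms emitter fires two small bursts in separate spawn groups, with group-masked scale, velocity and drag giving each wave distinct motion.
# Curves grow the meshes, fade their alpha and drive a dynamic material parameter alongside a randomized one, over a brightened gold-orange color.
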NiagaraSystem /BRCosmetics/Weapons/FORT_Melee/Pickaxe_Marshmellow_1h_B/FX/NS_Mello_Speaker_Gold_1H
dependencies {usedDataInterfaces [&NiagaraDataInterfaceCurve], usedDynamicInputs [], usedModules [&SystemState]}
emitters {Waveforms {
  emitterUpdateScript [
   &EmitterState()
   &SpawnBurst_Instantaneous(Spawn Count=2, Spawn Time=0.1, Use Spawn Probability=false, UseLoopCountLimit=false)
   &SpawnBurst_Instantaneous(Spawn Count=2, Spawn Group=1, Use Spawn Probability=false, UseLoopCountLimit=false) {module=SpawnBurst_Instantaneous001}
  ]
  particleSpawnScript [
   &InitializeParticle(
    Color=<rgba 1 0.4167 0 1>
    Lifetime=0.3
    Mesh Scale Mode=::NewEnumerator3
    Mesh Uniform Scale=&MaskFloatBySpawnGroup(Masked Float=12, Passthrough Float=6)
    UsePositionOffset=false
    Write Material Random=true
   )
   &InitialMeshOrientation(Mesh Axis To Orient=<0 0 1>, Mesh Orientation Mode=::NewEnumerator2, Use Rotation Vector=false)
   &AddVelocity(Velocity=&MaskVectorBySpawnGroup(Masked Vector=<60 0 0>, Passthrough Vector=<30 0 0>))
  ]
  particleUpdateScript [
   &Drag(
    Drag=&MaskFloatBySpawnGroup(Masked Float=4, Passthrough Float=0)
    Use Linear Drag=true
    Use Rotational Drag=false
   )
   &SolveForcesAndVelocity(Clamp Velocity=false, Limit Acceleration=false)
   &ParticleState()
   &ScaleMeshSize(Scale Factor=&VectorFromFloat(Value=&FloatFromCurve(FloatCurve=DataInterface() {propertyValues {Curve {keys [
          {arriveTangent=5.1547, interpMode=RCIM_Cubic, leaveTangent=5.1547, leaveTangentWeight=0.7454, tangentMode=RCTM_User, time=0, value=1}
          {arriveTangentWeight=0.7454, interpMode=RCIM_Cubic, tangentMode=RCTM_User, time=1, value=3}
         ]}}})))
   &DynamicMaterialParameters(
    Index 0 Param 1=0.5
    Index 0 Param 2=1
    Index 0 Param 3=&RandomRangeFloat(Minimum=-360, Maximum=360)
    Index 0 Param 4=&FloatFromCurve(FloatCurve=DataInterface() {propertyValues {Curve {keys [
         {arriveTangent=7.0476, interpMode=RCIM_Cubic, leaveTangent=7.0476, leaveTangentWeight=1.0541, tangentMode=RCTM_User, time=0, value=0}
         {arriveTangentWeight=1.0541, interpMode=RCIM_Cubic, tangentMode=RCTM_User, time=1, value=3}
        ]}}})
    Module.Param0WriteEnabled=true
    Module.Param1WriteEnabled=true
    Module.Param2WriteEnabled=true
    Module.Param3WriteEnabled=true
   )
   &ScaleColor(
    Scale Alpha=&FloatFromCurve(FloatCurve=DataInterface() {propertyValues {Curve {keys [
         {interpMode=RCIM_Cubic, leaveTangentWeight=0.4714, tangentMode=RCTM_User, time=0, value=1}
         {arriveTangentWeight=0.4714, interpMode=RCIM_Cubic, tangentMode=RCTM_User, time=1, value=0}
        ]}}})
    Scale RGB=<5 5 5>
    ScaleA=true
    ScaleRGB=true
   )
  ]
  rendererClasses [&NiagaraMeshRendererProperties]
  renderers [{rendererClass=&NiagaraMeshRendererProperties, rendererIndex=0}]
 }}
summary {
 systemName=NS_Mello_Speaker_Gold_1H
 emitters [{emitterName=Waveforms, rendererClasses [&NiagaraMeshRendererProperties]}]
 userVariables []
}
